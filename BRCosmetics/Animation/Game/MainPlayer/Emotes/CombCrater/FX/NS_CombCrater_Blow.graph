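# Niagara effect for the "Blow" moment of the CombCrater player emote: a purple puff of fire and embers pushed along +Y and slightly upward.
# A GPU ember burst uses drag, random gravity, speed-based sprite stretching and distance-based culling and spawn scaling.
# Two fire emitters animate sub-image sprites with platform-dependent camera offset, alongside a faint purple sprite that shrinks and fades.
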NiagaraSystem /BRCosmetics/Animation/Game/MainPlayer/Emotes/CombCrater/FX/NS_CombCrater_Blow
dependencies {
 usedDataInterfaces [&NiagaraDataInterfaceCurve, &NiagaraDataInterfaceColorCurve, &NiagaraDataInterfacePlatformSet]
 usedDynamicInputs [&SimulationPosition]
 usedModules [
  &self:SystemSpawnScript.SystemScriptSource.SystemScriptGraph.NiagaraNodeAssignment_0.SetVariables_#1
  &SystemState
 ]
}
emitters {
 EmberBurst001 {
  simTarget=GPUComputeSim
  emitterUpdateScript [
   &EmitterState(
    Enable Distance Culling=true
    Enable Visibility Culling=true
    Life Cycle Mode=::NewEnumerator1
    Loop Behavior=::NewEnumerator1
    Loop Duration=&RandomRangeFloat(Minimum=3.33, Maximum=6.33)
    MaxDistance=1000
    Module.bApplySpawnCountScaleByDistance=true
    Module.bUseMaxDistance=true
    Module.bUseMinDistance=false
    Scalability Mode=::NewEnumerator1
    Scale Spawn Count=true
    Spawn Count Scale By Distance=DataInterface() {propertyValues {Curve {keys [<at 0 1>,<at 0.3 1>,<at 1 0>]}}}
    UseLoopDelay=false
   )
   &SpawnBurst_Instantaneous(
    Spawn Count=&RandomRangeInt(Minimum=20, Maximum=30)
    Use Spawn Probability=false
    UseLoopCountLimit=false
   )
  ]
  particleSpawnScript [
   &InitializeParticle(
    Lifetime=&UniformRangedFloat(Minimum=0.3, Maximum=3)
    Sprite Size Mode=::NewEnumerator3
    Uniform Sprite Size=&MakeFloatFromVector2d(Vector 2D=&Vector2DFromFloat(Value=&FloatFromCurve(
       FloatCurve=DataInterface() {propertyValues {Curve {keys [<at 0 1.2>,<at 0.1257 0.9847>,<at 0.2996 0.5>,<at 1 0.3>]}}}
       CurveIndex=&UniformRangedFloat(Minimum=0, Maximum=1)
       Scale Curve=1.2
      )))
    UsePositionOffset=false
    Write Material Random=true
   ) {module=InitializeParticle001}
   &AddVelocity(
    Rotation Coordinate Space=::Simulation
    Velocity=&RandomRangeVector(Minimum=<-20 30 15>, Maximum=<20 30 15>)
    Velocity Speed Scale=2
   )
   &ShapeLocation(MaskSpawnGroup=false, Offset=<0 20 0>, Offset Mode=::NewEnumerator0, Override Seed=false, Sphere Radius=0, UseOwnerScale=true)
  ]
  particleUpdateScript [
   &ParticleState()
   &AccelerationForce(Acceleration=&Multiply_Vector(
     A=&VectorFromFloat(Value=&RandomRangeFloat(Minimum=0.75, Maximum=0.75))
     B=&UniformRangedVector(Minimum=<0 0 300>, Maximum=<0 0 300>)
    ))
   &Drag(
    Drag=&FloatFromCurve(FloatCurve=DataInterface() {propertyValues {Curve {keys [<at 0 0>,<at 0.1 0>,<at 1 1>]}}}, Scale Curve=50)
    Use Linear Drag=true
    Use Rotational Drag=false
   )
   /Niagara/Modules/Update/Forces/CurlNoiseForce(
    Noise Strength=&FloatFromCurve(
     FloatCurve=DataInterface() {propertyValues {Curve {keys [<at 0 1>,<at 0.1 0.25>,<at 0.1198 0.98>,<at 0.2002 0.21>,<at 0.2421 0.9996>,<at 1 0>]}}}
     Scale Curve=&RandomRangeFloat(Minimum=0, Maximum=0)
    )
    Pan Noise=true
    Pan Noise Field=<0 0 1>
   )
   &GravityForce(Gravity=&RandomRangeVector(Minimum=<0 0 0>, Maximum=<0 0 -250>))
   &SolveForcesAndVelocity(Clamp Velocity=false, Limit Acceleration=false)
   &ScaleSpriteSizeBySpeed(
    Max Scale Factor=<2 5>
    Sample Scale Factor By Curve=true
    Scale Factor Curve=DataInterface() {propertyValues {Curve {keys [
        <at -0.0077 0.0765 RCIM_Cubic>
        {arriveTangent=1.2463, interpMode=RCIM_Cubic, leaveTangent=1.2463, tangentMode=RCTM_User, time=0.9985, value=0.9549}
       ]}}}
    Velocity Threshold=250
   )
   &ScaleColor(
    Scale Alpha=&FloatFromCurve(FloatCurve=DataInterface() {propertyValues {Curve {keys [<at 0 1>,<at 1 0>]}}})
    Scale RGB=&MakeVectorFromLinearColorRGB(Color=&ColorFromCurve(ColorCurve=DataInterface() {propertyValues {
        AlphaCurve {keys [<at 0 1 RCIM_Cubic>,<at 0.7107 1>]}
        BlueCurve {keys [<at 0 7.8125 RCIM_Cubic>,<at 0.7114 1 RCIM_Cubic>]}
        GreenCurve {keys [<at 0 0 RCIM_Cubic>,<at 0.7114 0 RCIM_Cubic>]}
        RedCurve {keys [<at 0 1.9409 RCIM_Cubic>,<at 0.7114 0.8399 RCIM_Cubic>]}
       }}))
    ScaleA=true
    ScaleRGB=true
   )
   &ScaleSpriteSizeBySpeed(Max Scale Factor=<1 3>, Min Scale Factor=<0.5 0.5>, Velocity Threshold=350) {module=ScaleSpriteSizeBySpeed001}
  ]
  rendererClasses [&NiagaraSpriteRendererProperties]
  renderers [{rendererClass=&NiagaraSpriteRendererProperties}]
 }
 Fire_Particle {
  emitterUpdateScript [
   &EmitterState(Life Cycle Mode=::NewEnumerator1, Loop Behavior=::NewEnumerator1, UseLoopDelay=false)
   &SpawnBurst_Instantaneous(Spawn Count=2, Use Spawn Probability=false, UseLoopCountLimit=false)
  ]
  particleSpawnScript [
   &InitializeParticle(
    Color=<rgba 0.5263 0.1064 0.6458 1>
    Lifetime Max=0.6
    Lifetime Min=0.4
    Lifetime Mode=::NewEnumerator1
    Position Offset=<0 5 20>
    Sprite Rotation Mode=::NewEnumerator2
    Sprite Size Mode=::NewEnumerator3
    Uniform Sprite Size=&Multiply_Float(A=40, B=1.1)
    UsePositionOffset=true
    Write Material Random=true
   )
   &CameraOffset(Camera Offset Amount=&FN_PickFloatDependOnPlatform(TargetPlatformActive=0, TargetPlatformNoActive=2))
   &AddVelocity(Velocity=&RandomRangeVector(Minimum=<0 40 2>, Maximum=<0 40 2>), Velocity Speed Scale=2)
   &ShapeLocation(MaskSpawnGroup=false, Offset=<0 15 -24>, Offset Mode=::NewEnumerator0, Override Seed=false, Sphere Radius=0, UseOwnerScale=true)
  ]
  particleUpdateScript [
   &ParticleState()
   /SetVariables(Particles.SubImageIndex=&FloatFromCurve(FloatCurve=DataInterface() {propertyValues {Curve {keys [<at 0 5>,<at 1 15>]}}})) {isSetParameters=true, module=SetVariables_#2}
   &ScaleSpriteSize(Scale Factor=&Vector2DFromFloat(Value=&FloatFromCurve(
      FloatCurve=DataInterface() {propertyValues {Curve {keys [
          {interpMode=RCIM_Cubic, leaveTangent=3.0564, tangentMode=RCTM_Break, time=0, value=0}
          {arriveTangent=1.1454, interpMode=RCIM_Cubic, leaveTangent=1.1454, tangentMode=RCTM_User, time=0.3, value=0.8}
          {arriveTangent=-1.5127, interpMode=RCIM_Cubic, tangentMode=RCTM_Break, time=1, value=0.6}
         ]}}}
      Scale Curve=0.8
     )))
   &ScaleColor(
    Scale Alpha=&FloatFromCurve(FloatCurve=DataInterface() {propertyValues {Curve {keys [<at 0 0 RCIM_Cubic>,<at 0.3 1 RCIM_Cubic>,<at 1 0 RCIM_Cubic>]}}})
    Scale RGB=&MakeVectorFromLinearColorRGB(Color=&ColorFromCurve(ColorCurve=DataInterface() {propertyValues {
        AlphaCurve {keys [<at 0 0 RCIM_Cubic>,<at 0.1107 1>]}
        BlueCurve {keys [<at 0 1 RCIM_Cubic>,<at 0.2918 1>,<at 1 1 RCIM_Cubic>]}
        GreenCurve {keys [<at 0 0 RCIM_Cubic>,<at 0.2918 0>,<at 1 0 RCIM_Cubic>]}
        RedCurve {keys [<at 0 0.5041 RCIM_Cubic>,<at 0.2918 0.5029>,<at 1 0.0264 RCIM_Cubic>]}
       }}))
    ScaleA=true
    ScaleRGB=true
   )
   &SolveForcesAndVelocity(Clamp Velocity=false, Limit Acceleration=false)
  ]
  rendererClasses [&NiagaraSpriteRendererProperties]
  renderers [{rendererClass=&NiagaraSpriteRendererProperties}]
 }
 Fire_Particle001 {
  emitterUpdateScript [
   &EmitterState(Life Cycle Mode=::NewEnumerator1, Loop Behavior=::NewEnumerator1, UseLoopDelay=false)
   &SpawnBurst_Instantaneous(Spawn Count=2, Use Spawn Probability=false, UseLoopCountLimit=false)
  ]
  particleSpawnScript [
   &InitializeParticle(
    Color=<rgba 0.2676 0.0004 0.75 1>
    Lifetime Max=0.7
    Lifetime Min=0.5
    Lifetime Mode=::NewEnumerator1
    Position Offset=<0 5 20>
    Sprite Rotation Mode=::NewEnumerator2
    Sprite Size Mode=::NewEnumerator3
    Uniform Sprite Size=&Multiply_Float(A=40, B=1.1)
    UsePositionOffset=true
    Write Material Random=true
   )
   &CameraOffset(Camera Offset Amount=&FN_PickFloatDependOnPlatform(TargetPlatformActive=0, TargetPlatformNoActive=2))
   &AddVelocity(Velocity=&RandomRangeVector(Minimum=<0 40 2>, Maximum=<0 40 2>), Velocity Speed Scale=2)
   &ShapeLocation(MaskSpawnGroup=false, Offset=<0 15 -24>, Offset Mode=::NewEnumerator0, Override Seed=false, Sphere Radius=0, UseOwnerScale=true)
  ]
  particleUpdateScript [
   &ParticleState()
   /SetVariables(Particles.SubImageIndex=&FloatFromCurve(FloatCurve=DataInterface() {propertyValues {Curve {keys [<at 0 5>,<at 1 64>]}}})) {isSetParameters=true, module=SetVariables_#2}
   &ScaleSpriteSize(Scale Factor=&Vector2DFromFloat(Value=&FloatFromCurve(
      FloatCurve=DataInterface() {propertyValues {Curve {keys [
          {interpMode=RCIM_Cubic, leaveTangent=3.0564, tangentMode=RCTM_Break, time=0, value=0}
          {arriveTangent=1.1454, interpMode=RCIM_Cubic, leaveTangent=1.1454, tangentMode=RCTM_User, time=0.3, value=0.8}
          {arriveTangent=-1.5127, interpMode=RCIM_Cubic, tangentMode=RCTM_Break, time=1, value=0.6}
         ]}}}
      Scale Curve=0.8
     )))
   &ScaleColor(
    Scale Alpha=&FloatFromCurve(FloatCurve=DataInterface() {propertyValues {Curve {keys [<at 0 0 RCIM_Cubic>,<at 0.3 1 RCIM_Cubic>,<at 1 0 RCIM_Cubic>]}}})
    Scale RGB=&MakeVectorFromLinearColorRGB(Color=&ColorFromCurve(ColorCurve=DataInterface() {propertyValues {
        AlphaCurve {keys [<at 0 1 RCIM_Cubic>]}
        BlueCurve {keys [<at 0 1 RCIM_Cubic>,<at 1 1 RCIM_Cubic>]}
        GreenCurve {keys [<at 0 1 RCIM_Cubic>,<at 1 0.0039 RCIM_Cubic>]}
        RedCurve {keys [<at 0 1 RCIM_Cubic>,<at 1 0.6824 RCIM_Cubic>]}
       }}))
    ScaleA=true
    ScaleRGB=true
   )
   &SolveForcesAndVelocity(Clamp Velocity=false, Limit Acceleration=false)
  ]
  rendererClasses [&NiagaraSpriteRendererProperties]
  renderers [{rendererClass=&NiagaraSpriteRendererProperties}]
 }
 SingleLoopingParticle001 {
  emitterUpdateScript [
   &EmitterState(Life Cycle Mode=::NewEnumerator1, Loop Behavior=::NewEnumerator1, Loop Duration=1, UseLoopDelay=false)
   &SpawnBurst_Instantaneous(Use Spawn Probability=false, UseLoopCountLimit=false)
  ]
  particleSpawnScript [
   &InitializeParticle(
    Color=<rgba 0.4837 0.0275 1 0.1>
    Lifetime=0.8
    Position Offset=<0 5 0>
    Sprite Size Mode=::NewEnumerator3
    Uniform Sprite Size=30
    UsePositionOffset=true
    Write Material Random=true
   )
   &ShapeLocation(MaskSpawnGroup=false, Offset=<0 15 -6>, Offset Mode=::NewEnumerator0, Override Seed=false, Sphere Radius=0, UseOwnerScale=true)
   &AddVelocity(Velocity=&RandomRangeVector(Minimum=<0 45 15>, Maximum=<0 45 15>), Velocity Speed Scale=2)
  ]
  particleUpdateScript [
   &ParticleState()
   &ScaleSpriteSize(Uniform Curve Sprite Scale=DataInterface() {propertyValues {Curve {keys [{tangentMode=RCTM_User, time=0, value=1}, {tangentMode=RCTM_User, time=1, value=0}]}}})
   &SolveForcesAndVelocity(Clamp Velocity=false, Limit Acceleration=false)
   &ScaleColor(
    Scale Alpha=&FloatFromCurve(FloatCurve=DataInterface() {propertyValues {Curve {keys [<at 0 1>,<at 1 0>]}}})
    ScaleA=true
    ScaleRGB=true
   )
  ]
  rendererClasses [&NiagaraSpriteRendererProperties]
  renderers [{rendererClass=&NiagaraSpriteRendererProperties}]
 }
}
summary {
 systemName=NS_CombCrater_Blow
 emitters [
  {emitterName=EmberBurst001, simTarget=GPUComputeSim, rendererClasses [&NiagaraSpriteRendererProperties]}
  {emitterName=SingleLoopingParticle001, rendererClasses [&NiagaraSpriteRendererProperties]}
  {emitterName=Fire_Particle, rendererClasses [&NiagaraSpriteRendererProperties]}
  {emitterName=Fire_Particle001, rendererClasses [&NiagaraSpriteRendererProperties]}
 ]
 userVariables [User.FXScale {type=NiagaraFloat, defaultValue=1, description=""}]
}
