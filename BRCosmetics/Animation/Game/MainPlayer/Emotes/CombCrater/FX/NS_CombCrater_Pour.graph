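# Niagara effect for the pour in the CombCrater player emote: a short, glowing purple stream falling from the system's location.
# Trail_Water003 draws the ribbon, re-placing particles at the system location plus an age-driven downward offset each update, so the stream follows its source without gravity.
# Fire_Particle adds slightly delayed, vertically stretched flipbook sprites at fixed-seed random offsets along it.
# The other trail and wispy smoke emitters are disabled.
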NiagaraSystem /BRCosmetics/Animation/Game/MainPlayer/Emotes/CombCrater/FX/NS_CombCrater_Pour
dependencies {
 usedDataInterfaces [
  &NiagaraDataInterfaceColorCurve
  &NiagaraDataInterfaceCurve
  &NiagaraDataInterfaceVectorCurve
  &NiagaraDataInterfaceVector2DCurve
  &NiagaraDataInterfaceSpriteRendererInfo
  &NiagaraDataInterfacePlatformSet
 ]
 usedDynamicInputs [&Vector2DFromCurve]
 usedModules [
  &SystemState
  &self:SystemSpawnScript.SystemScriptSource.SystemScriptGraph.NiagaraNodeAssignment_0.SetVariables_#1
  &self:SystemSpawnScript.SystemScriptSource.SystemScriptGraph.NiagaraNodeAssignment_1.SetVariables_#2
 ]
}
emitters {
 Fire_Particle {
  emitterUpdateScript [
   &EmitterState(Life Cycle Mode=::NewEnumerator1, Loop Behavior=::NewEnumerator1, Loop Delay=0.2, Loop Duration=1.2, UseLoopDelay=true)
   &SpawnRate(SpawnRate=50, Use Spawn Probability=false)
  ]
  particleSpawnScript [
   &InitializeParticle(
    Color=<rgba 1.2982 0.5454 3 1>
    Lifetime Max=1
    Lifetime Min=0.75
    Lifetime Mode=::NewEnumerator1
    Sprite Size Mode=::NewEnumerator1
    Uniform Sprite Size Max=15
    Uniform Sprite Size Min=10
    UsePositionOffset=false
    Write Material Random=true
   )
   &CameraOffset(Camera Offset Amount=&FN_PickFloatDependOnPlatform(TargetPlatformActive=0, TargetPlatformNoActive=2)) {enabled=false}
  ]
  particleUpdateScript [
   &ParticleState()
   /SetVariables(Particles.SubImageIndex=&FloatFromCurve(FloatCurve=DataInterface() {propertyValues {Curve {keys [<at 0 5>,<at 1 15>]}}})) {isSetParameters=true}
   &SystemLocation(
    MaskSpawnGroup=false
    Offset=&RandomRangeVector(Minimum=<-1 -1 -50>, Maximum=<1 1 -150>, Random Seed=5, Fixed Random Seed=true)
   ) {module=SystemLocation001}
   &ScaleSpriteSize(Scale Factor=&MakeVector2D(
     X=&FloatFromCurve(
      FloatCurve=DataInterface() {propertyValues {Curve {keys [
          {interpMode=RCIM_Cubic, leaveTangent=0.1472, tangentMode=RCTM_Break, time=0, value=0}
          {arriveTangent=2.3722, interpMode=RCIM_Cubic, leaveTangent=2.3722, tangentMode=RCTM_User, time=0.3, value=0.8}
          {arriveTangent=-1.5127, interpMode=RCIM_Cubic, tangentMode=RCTM_Break, time=1, value=1}
         ]}}}
      Scale Curve=&RandomRangeFloat(Minimum=0.2, Maximum=0.6)
     )
     Y=&FloatFromCurve(
      FloatCurve=DataInterface() {propertyValues {Curve {keys [
          {interpMode=RCIM_Cubic, leaveTangent=0.1472, tangentMode=RCTM_Break, time=0, value=0}
          {arriveTangent=2.3722, interpMode=RCIM_Cubic, leaveTangent=2.3722, tangentMode=RCTM_User, time=0.3, value=0.8}
          {arriveTangent=-1.5127, interpMode=RCIM_Cubic, tangentMode=RCTM_Break, time=1, value=0.6}
         ]}}}
      Scale Curve=&RandomRangeFloat(Minimum=2, Maximum=3)
     )
    ))
   &ScaleColor(
    Scale Alpha=&FloatFromCurve(
     FloatCurve=DataInterface() {propertyValues {Curve {keys [<at 0 0 RCIM_Cubic>,<at 0.3 1 RCIM_Cubic>,<at 0.8 0 RCIM_Cubic>]}}}
     Scale Curve=&RandomRangeFloat(Minimum=0.6, Maximum=1)
    )
    Scale RGB=&MakeVectorFromLinearColorRGB(Color=&ColorFromCurve(ColorCurve=DataInterface() {propertyValues {
        AlphaCurve {keys [<at 0 1 RCIM_Cubic>,<at 1 0 RCIM_Cubic>]}
        BlueCurve {keys [<at 0 3 RCIM_Cubic>,<at 1 0.25 RCIM_Cubic>]}
        GreenCurve {keys [<at 0 0.597 RCIM_Cubic>,<at 1 0.25 RCIM_Cubic>]}
        RedCurve {keys [<at 0 2.5359 RCIM_Cubic>,<at 1 0.25 RCIM_Cubic>]}
       }}))
    ScaleA=true
    ScaleRGB=true
   )
  ]
  rendererClasses [&NiagaraSpriteRendererProperties]
  renderers [{rendererClass=&NiagaraSpriteRendererProperties}]
 }
 Trail_Water {
  bEnabled=false
  emitterUpdateScript [
   &EmitterState(Life Cycle Mode=::NewEnumerator1, Loop Behavior=::NewEnumerator1, Loop Duration=1.4, UseLoopDelay=false)
   &SpawnRate(SpawnRate=30, Use Spawn Probability=false)
  ]
  particleSpawnScript [
   &InitializeParticle(
    Color=&ColorFromCurve(
     ColorCurve=DataInterface() {propertyValues {
       AlphaCurve {keys [<at 0 0 RCIM_Cubic>,<at 0.2181 0.8>,<at 0.7961 0.8>,<at 1 0 RCIM_Cubic>]}
       BlueCurve {keys [<at 0 1 RCIM_Cubic>,<at 1 3 RCIM_Cubic>]}
       GreenCurve {keys [<at 0 0.1342 RCIM_Cubic>,<at 1 0.0938 RCIM_Cubic>]}
       RedCurve {keys [<at 0 0.7 RCIM_Cubic>,<at 1 0.8 RCIM_Cubic>]}
      }}
     CurveIndex=$Emitter.NormalizedLoopAge
    )
    Lifetime=0.62
    Ribbon Width=4
    Ribbon Width Mode=::NewEnumerator1
    UsePositionOffset=false
    Write Material Random=true
   )
   &ShapeLocation(MaskSpawnGroup=false, Offset Mode=::NewEnumerator0, Override Seed=false, Sphere Radius=0, UseOwnerScale=true)
   &AddVelocity(Velocity=<0 0 25>)
  ]
  particleUpdateScript [
   &InheritSourceMovement(
    Applied Position Scale=&Multiply_VectorByFloat(
     Vector=<1 1 1>
     Float=&FloatFromCurve(FloatCurve=DataInterface() {propertyValues {Curve {keys [<at 0 1>,<at 0.6 1>,<at 0.61 0>]}}}, CurveIndex=$Emitter.NormalizedLoopAge)
    )
    ApplyForce=false
    ApplyPosition=true
    ApplyPositionFalloff=false
   ) {enabled=false}
   &GravityForce(Gravity=<0 0 -680>)
   &SolveForcesAndVelocity(Clamp Velocity=false, Limit Acceleration=false)
   &ParticleState()
   &ScaleColor(
    Scale Alpha=&FloatFromCurve(FloatCurve=DataInterface() {propertyValues {Curve {keys [
         {interpMode=RCIM_Cubic, tangentMode=RCTM_User, time=0, value=0}
         {interpMode=RCIM_Cubic, tangentMode=RCTM_User, time=0.7, value=1}
         {tangentMode=RCTM_User, time=1, value=0}
        ]}}})
    ScaleA=true
    ScaleRGB=true
   )
  ]
  rendererClasses [&NiagaraRibbonRendererProperties]
  renderers [{rendererClass=&NiagaraRibbonRendererProperties}]
 }
 Trail_Water001 {
  bEnabled=false
  emitterUpdateScript [
   &EmitterState(Life Cycle Mode=::NewEnumerator1, Loop Behavior=::NewEnumerator1, Loop Duration=1.4, UseLoopDelay=false)
   &SpawnRate(SpawnRate=15, Use Spawn Probability=false)
  ]
  particleSpawnScript [
   &InitializeParticle(
    Color=&ColorFromCurve(
     ColorCurve=DataInterface() {propertyValues {
       AlphaCurve {keys [<at 0 0 RCIM_Cubic>,<at 0.1307 1>,<at 0.858 2>,<at 1 0 RCIM_Cubic>]}
       BlueCurve {keys [<at 0 1 RCIM_Cubic>,<at 1 1 RCIM_Cubic>]}
       GreenCurve {keys [<at 0 0.1144 RCIM_Cubic>,<at 1 0.1111 RCIM_Cubic>]}
       RedCurve {keys [<at 0 0.314 RCIM_Cubic>,<at 1 0.3036 RCIM_Cubic>]}
      }}
     CurveIndex=$Emitter.NormalizedLoopAge
    )
    Lifetime=0.62
    Ribbon Width=20
    Ribbon Width Mode=::NewEnumerator1
    UsePositionOffset=false
    Write Material Random=true
   )
   &AddVelocity(Velocity=<5 5 50>)
  ]
  particleUpdateScript [
   &GravityForce(Gravity=<0 0 -680>)
   &SolveForcesAndVelocity(Clamp Velocity=false, Limit Acceleration=false)
   &ParticleState()
   &ScaleColor(
    Scale Alpha=&FloatFromCurve(
     FloatCurve=DataInterface() {propertyValues {Curve {keys [
         {interpMode=RCIM_Cubic, tangentMode=RCTM_User, time=0, value=0}
         {interpMode=RCIM_Cubic, tangentMode=RCTM_User, time=0.7, value=1}
         {tangentMode=RCTM_User, time=1, value=0}
        ]}}}
     Scale Curve=0.5
    )
    ScaleA=true
    ScaleRGB=true
   )
  ]
  rendererClasses [&NiagaraRibbonRendererProperties]
  renderers [{rendererClass=&NiagaraRibbonRendererProperties}]
 }
 Trail_Water002 {
  bEnabled=false
  emitterUpdateScript [
   &EmitterState(Life Cycle Mode=::NewEnumerator1, Loop Behavior=::NewEnumerator1, Loop Duration=1.4, UseLoopDelay=false)
   &SpawnRate(SpawnRate=15, Use Spawn Probability=false)
  ]
  particleSpawnScript [
   &InitializeParticle(
    Color=&ColorFromCurve(
     ColorCurve=DataInterface() {propertyValues {
       AlphaCurve {keys [<at 0 0 RCIM_Cubic>,<at 0.0739 2>,<at 0.9289 2>,<at 1 0 RCIM_Cubic>]}
       BlueCurve {keys [<at 0 1 RCIM_Cubic>,<at 1 1 RCIM_Cubic>]}
       GreenCurve {keys [<at 0 0.1144 RCIM_Cubic>,<at 1 0.1044 RCIM_Cubic>]}
       RedCurve {keys [<at 0 0.314 RCIM_Cubic>,<at 1 0.3103 RCIM_Cubic>]}
      }}
     CurveIndex=$Emitter.NormalizedLoopAge
    )
    Lifetime=0.62
    Ribbon Width=15
    Ribbon Width Mode=::NewEnumerator1
    UsePositionOffset=false
    Write Material Random=true
   )
   &ShapeLocation(MaskSpawnGroup=false, Offset Mode=::NewEnumerator0, Override Seed=false, Sphere Radius=0, UseOwnerScale=true)
   &AddVelocity(Velocity=<0 0 25>)
  ]
  particleUpdateScript [
   &InheritSourceMovement(
    Applied Position Scale=&Multiply_VectorByFloat(
     Vector=<1 1 1>
     Float=&FloatFromCurve(FloatCurve=DataInterface() {propertyValues {Curve {keys [<at 0 1>,<at 0.6 1>,<at 0.61 0>]}}}, CurveIndex=$Emitter.NormalizedLoopAge)
    )
    ApplyForce=false
    ApplyPosition=true
    ApplyPositionFalloff=false
   ) {enabled=false}
   &GravityForce(Gravity=<0 0 -680>)
   &SolveForcesAndVelocity(Clamp Velocity=false, Limit Acceleration=false)
   &ParticleState()
   &ScaleColor(
    Scale Alpha=&FloatFromCurve(
     FloatCurve=DataInterface() {propertyValues {Curve {keys [
         {interpMode=RCIM_Cubic, tangentMode=RCTM_User, time=0, value=0}
         {interpMode=RCIM_Cubic, tangentMode=RCTM_User, time=0.7, value=1}
         {tangentMode=RCTM_User, time=1, value=0}
        ]}}}
     Scale Curve=0.1
    )
    ScaleA=true
    ScaleRGB=true
   )
   &ScaleRibbonWidth(Ribbon Width Scale=&FloatFromCurve(FloatCurve=DataInterface() {propertyValues {Curve {keys [<at 0 1>,<at 1 1.5>]}}}))
  ]
  rendererClasses [&NiagaraRibbonRendererProperties]
  renderers [{rendererClass=&NiagaraRibbonRendererProperties}]
 }
 Trail_Water003 {
  emitterUpdateScript [
   &EmitterState(Life Cycle Mode=::NewEnumerator1, Loop Behavior=::NewEnumerator1, Loop Duration=1.2, UseLoopDelay=false)
   &SpawnRate(SpawnRate=30, Use Spawn Probability=false)
  ]
  particleSpawnScript [
   &InitializeParticle(
    Color=&ColorFromCurve(
     ColorCurve=DataInterface() {propertyValues {
       AlphaCurve {keys [<at 0 0 RCIM_Cubic>,<at 0.0423 0.8>,<at 0.4619 0.8>,<at 0.9812 0.8>,<at 1 0 RCIM_Cubic>]}
       BlueCurve {keys [<at 0 1 RCIM_Cubic>,<at 1 3 RCIM_Cubic>]}
       GreenCurve {keys [<at 0 0.1342 RCIM_Cubic>,<at 1 0.0938 RCIM_Cubic>]}
       RedCurve {keys [<at 0 0.7 RCIM_Cubic>,<at 1 0.8 RCIM_Cubic>]}
      }}
     CurveIndex=$Emitter.NormalizedLoopAge
    )
    Lifetime=0.6
    Ribbon Width=2
    Ribbon Width Mode=::NewEnumerator1
    UsePositionOffset=false
    Write Material Random=true
   )
   &ShapeLocation(MaskSpawnGroup=false, Offset Mode=::NewEnumerator0, Override Seed=false, Sphere Radius=0, UseOwnerScale=true)
   &AddVelocity(Velocity=<0 0 25>) {enabled=false}
   /SetVariables(Particles.RibbonColor02:LinearColor=&ColorFromCurve(
     ColorCurve=DataInterface() {propertyValues {
       AlphaCurve {keys [<at 0 0 RCIM_Cubic>,<at 0.0739 2>,<at 0.9289 2>,<at 1 0 RCIM_Cubic>]}
       BlueCurve {keys [<at 0 1 RCIM_Cubic>,<at 1 1 RCIM_Cubic>]}
       GreenCurve {keys [<at 0 0.1144 RCIM_Cubic>,<at 1 0.1044 RCIM_Cubic>]}
       RedCurve {keys [<at 0 0.314 RCIM_Cubic>,<at 1 0.3103 RCIM_Cubic>]}
      }}
     CurveIndex=$Emitter.NormalizedLoopAge
    )) {enabled=false, isSetParameters=true}
  ]
  particleUpdateScript [
   &InheritSourceMovement(
    Applied Position Scale=&Multiply_VectorByFloat(
     Vector=<1 1 1>
     Float=&FloatFromCurve(FloatCurve=DataInterface() {propertyValues {Curve {keys [<at 0 1>,<at 0.6 1>,<at 0.61 0>]}}}, CurveIndex=$Emitter.NormalizedLoopAge)
    )
    ApplyForce=false
    ApplyPosition=true
    ApplyPositionFalloff=false
   ) {enabled=false}
   &GravityForce(Gravity=<0 0 -680>) {enabled=false}
   &SolveForcesAndVelocity(Clamp Velocity=false, Limit Acceleration=false)
   &ParticleState()
   &ScaleColor(
    Scale Alpha=&FloatFromCurve(FloatCurve=DataInterface() {propertyValues {Curve {keys [
         {interpMode=RCIM_Cubic, tangentMode=RCTM_User, time=0, value=0}
         {arriveTangent=0.3339, interpMode=RCIM_Cubic, leaveTangent=0.3339, tangentMode=RCTM_User, time=0.05, value=0}
         {interpMode=RCIM_Cubic, tangentMode=RCTM_User, time=0.4, value=1}
         {arriveTangent=-0.6522, tangentMode=RCTM_User, time=1, value=0}
        ]}}})
    ScaleA=true
    ScaleRGB=true
   )
   &SystemLocation(
    MaskSpawnGroup=false
    Offset=&AddVectorToPosition(
     Position=&VectorFromCurve(
      VectorCurve=DataInterface() {propertyValues {XCurve {keys [<at 0 0>,<at 1 -1>]}, YCurve {keys [<at 0 0>,<at 1 -1>]}, ZCurve {keys [<at 0 0>,<at 1 -1>]}}}
      Scale Curve=<1 1 150>
     )
     Vector=<0 0 0>
    )
   )
   &ScaleRibbonWidth(Ribbon Width Scale=&FloatFromCurve(FloatCurve=DataInterface() {propertyValues {Curve {keys [<at 0 1>,<at 1 1.5>]}}})) {enabled=false}
   /SetVariables(
    Particles.RibbonColor02:LinearColor=&MultiplyLinearColor(
     Linear Color=$Particles.RibbonColor02
     Scale Factor=&MakeVector4(
      X=10
      Y=10
      Z=10
      W=&FloatFromCurve(FloatCurve=DataInterface() {propertyValues {Curve {keys [
           {interpMode=RCIM_Cubic, tangentMode=RCTM_User, time=0, value=0}
           {arriveTangent=2.5, interpMode=RCIM_Cubic, leaveTangent=2.5, time=0.12, value=0.05}
           {interpMode=RCIM_Cubic, tangentMode=RCTM_User, time=0.4, value=1}
           {arriveTangent=-0.6522, tangentMode=RCTM_User, time=1, value=0}
          ]}}})
     )
    )
    Particles.RibbonWidth02=15
   ) {enabled=false, isSetParameters=true}
  ]
  rendererClasses [&NiagaraRibbonRendererProperties]
  renderers [{rendererClass=&NiagaraRibbonRendererProperties}]
 }
 Trail_Water004 {
  bEnabled=false
  emitterUpdateScript [
   &EmitterState(Life Cycle Mode=::NewEnumerator1, Loop Behavior=::NewEnumerator1, Loop Duration=1.4, UseLoopDelay=false)
   &SpawnRate(SpawnRate=15, Use Spawn Probability=false)
  ]
  particleSpawnScript [
   &InitializeParticle(
    Color=&ColorFromCurve(
     ColorCurve=DataInterface() {propertyValues {
       AlphaCurve {keys [<at 0 0 RCIM_Cubic>,<at 0.0739 2>,<at 0.9289 2>,<at 1 0 RCIM_Cubic>]}
       BlueCurve {keys [<at 0 1 RCIM_Cubic>,<at 1 1 RCIM_Cubic>]}
       GreenCurve {keys [<at 0 0.1144 RCIM_Cubic>,<at 1 0.1044 RCIM_Cubic>]}
       RedCurve {keys [<at 0 0.314 RCIM_Cubic>,<at 1 0.3103 RCIM_Cubic>]}
      }}
     CurveIndex=$Emitter.NormalizedLoopAge
    )
    Lifetime=0.62
    Ribbon Width=15
    Ribbon Width Mode=::NewEnumerator1
    UsePositionOffset=false
    Write Material Random=true
   )
   &ShapeLocation(MaskSpawnGroup=false, Offset Mode=::NewEnumerator0, Override Seed=false, Sphere Radius=0, UseOwnerScale=true)
   &AddVelocity(Velocity=<0 0 25>) {enabled=false}
  ]
  particleUpdateScript [
   &InheritSourceMovement(
    Applied Position Scale=&Multiply_VectorByFloat(
     Vector=<1 1 1>
     Float=&FloatFromCurve(FloatCurve=DataInterface() {propertyValues {Curve {keys [<at 0 1>,<at 0.6 1>,<at 0.61 0>]}}}, CurveIndex=$Emitter.NormalizedLoopAge)
    )
    ApplyForce=false
    ApplyPosition=true
    ApplyPositionFalloff=false
   ) {enabled=false}
   &GravityForce(Gravity=<0 0 -680>) {enabled=false}
   &SolveForcesAndVelocity(Clamp Velocity=false, Limit Acceleration=false)
   &ParticleState()
   &ScaleColor(
    Scale Alpha=&FloatFromCurve(
     FloatCurve=DataInterface() {propertyValues {Curve {keys [
         {interpMode=RCIM_Cubic, tangentMode=RCTM_User, time=0, value=0}
         {interpMode=RCIM_Cubic, tangentMode=RCTM_User, time=0.7, value=1}
         {tangentMode=RCTM_User, time=1, value=0}
        ]}}}
     Scale Curve=0.1
    )
    ScaleA=true
    ScaleRGB=true
   )
   &ScaleRibbonWidth(Ribbon Width Scale=&FloatFromCurve(FloatCurve=DataInterface() {propertyValues {Curve {keys [<at 0 1>,<at 1 1.5>]}}}))
   &SystemLocation(
    MaskSpawnGroup=false
    Offset=&VectorFromCurve(
     VectorCurve=DataInterface() {propertyValues {XCurve {keys [<at 0 0>,<at 1 -1>]}, YCurve {keys [<at 0 0>,<at 1 -1>]}, ZCurve {keys [<at 0 0>,<at 1 -1>]}}}
     Scale Curve=<1 1 150>
    )
   )
  ]
  rendererClasses [&NiagaraRibbonRendererProperties]
  renderers [{rendererClass=&NiagaraRibbonRendererProperties}]
 }
 WispySmoke002 {
  bEnabled=false
  emitterUpdateScript [
   &EmitterState(Life Cycle Mode=::NewEnumerator1, Loop Behavior=::NewEnumerator1, Loop Delay=0.2, Loop Duration=1.8, UseLoopDelay=true)
   &SpawnRate(SpawnRate=50, Use Spawn Probability=false)
  ]
  particleSpawnScript [&InitializeParticle(
    Color=&ColorFromCurve(
     ColorCurve=DataInterface() {propertyValues {
       AlphaCurve {keys [<at 0 0 RCIM_Cubic>,<at 0.1064 0.5>,<at 0.8661 0.5>,<at 1 0 RCIM_Cubic>]}
       BlueCurve {keys [<at 0 1.5 RCIM_Cubic>,<at 1 1 RCIM_Cubic>]}
       GreenCurve {keys [<at 0 0 RCIM_Cubic>,<at 1 0 RCIM_Cubic>]}
       RedCurve {keys [<at 0 0.4181 RCIM_Cubic>,<at 1 0.3446 RCIM_Cubic>]}
      }}
     CurveIndex=$Emitter.NormalizedLoopAge
    )
    Lifetime Max=0.6
    Lifetime Min=0.4
    Lifetime Mode=::NewEnumerator1
    Sprite Size Max=&Lerp_Vector2(A=<-30 25>, B=<-25 30>, Alpha=&RandomRangeFloat(Minimum=0, Maximum=1))
    Sprite Size Min=&Lerp_Vector2(A=<15 20>, B=<15 20>, Alpha=&RandomRangeFloat(Minimum=0, Maximum=1))
    Sprite Size Mode=::NewEnumerator2
    UsePositionOffset=false
    Write Material Random=true
   )]
  particleUpdateScript [
   &InheritSourceMovement(
    Applied Position Scale=&Multiply_VectorByFloat(
     Vector=<1 1 1>
     Float=&FloatFromCurve(FloatCurve=DataInterface() {propertyValues {Curve {keys [<at 0 1>,<at 0.6 1>,<at 0.61 0>]}}}, CurveIndex=$Emitter.NormalizedLoopAge)
    )
    ApplyForce=false
    ApplyPosition=true
    ApplyPositionFalloff=false
   ) {enabled=false}
   &ParticleState()
   &ScaleSpriteSize(
    Non-Uniform Curve Scale=&RandomRangeVector2D(Minimum=<1 1>, Maximum=<1.2 1>)
    Non-Uniform Curve Sprite Scale=DataInterface() {propertyValues {
      XCurve {keys [{tangentMode=RCTM_User, time=0, value=0.9}, {tangentMode=RCTM_User, time=1, value=0.8}]}
      YCurve {keys [{tangentMode=RCTM_User, time=0, value=0.25}, <at 0.4 0.5>, {tangentMode=RCTM_User, time=1, value=1.175}]}
     }}
    Scale Sprite Size Mode=::NewEnumerator3
   )
   &SubUVAnimation(
    End Frame Range Override=30
    "Sprite Renderer "=DataInterface() {propertyValues {SpriteRenderer=&self:WispySmoke002_0.NiagaraSpriteRendererProperties_0}}
    Start Frame Range Override=&RandomRangeInt(Minimum=0, Maximum=5)
    SubUV Lookup Index:NiagaraFloat=&Lerp_Float(
     A=&RandomRangeFloat(Minimum=0, Maximum=0.2)
     B=&RandomRangeFloat(Minimum=0.3, Maximum=1)
     Alpha=$Particles.NormalizedAge
    )
    UseEndFrame=true
    UseStartFrame=true
   )
   &ScaleColor(
    Scale Alpha=&Multiply_Float(
     A=&RandomRangeFloat(Minimum=0.5, Maximum=1)
     B=&FloatFromCurve(
      FloatCurve=DataInterface() {propertyValues {Curve {keys [
          {arriveTangent=0.9483, interpMode=RCIM_Cubic, leaveTangent=0.9483, tangentMode=RCTM_User, time=0, value=0}
          {arriveTangent=2, interpMode=RCIM_Cubic, leaveTangent=2, time=0.1, value=0.05}
          {interpMode=RCIM_Cubic, tangentMode=RCTM_User, time=0.5, value=1}
          {arriveTangent=-5.5093, interpMode=RCIM_Cubic, leaveTangent=-5.5093, tangentMode=RCTM_User, time=1, value=0}
         ]}}}
      Scale Curve=&FloatFromCurve(
       FloatCurve=DataInterface() {propertyValues {Curve {keys [
           {tangentMode=RCTM_User, time=0, value=1}
           {arriveTangent=-0.0405, interpMode=RCIM_Cubic, leaveTangent=-0.0405, tangentMode=RCTM_User, time=1.5916, value=1}
           {arriveTangent=-2.5083, interpMode=RCIM_Cubic, leaveTangent=-2.5083, tangentMode=RCTM_User, time=2.4445, value=0}
          ]}}}
       CurveIndex=$Emitter.Age
      )
     )
    )
    ScaleA=true
    ScaleRGB=true
   )
   &SystemLocation(
    MaskSpawnGroup=false
    Offset=&RandomRangeVector(
     Minimum=<0 0 -40>
     Maximum=&Lerp_Vector(
      A=<0 0 -40>
      B=<1 1 -120>
      Alpha=&FloatFromCurve(FloatCurve=DataInterface() {propertyValues {Curve {keys [<at 0 0>,<at 0.3 0.15>,<at 1 1>]}}}, CurveIndex=$Emitter.Age)
     )
    )
   ) {enabled=false}
   &SystemLocation(
    MaskSpawnGroup=false
    Offset=&RandomRangeVector(Minimum=<0 0 -50>, Maximum=<0 0 -150>, Random Seed=5, Fixed Random Seed=true)
   ) {module=SystemLocation001}
   &GravityForce(Gravity=&RandomRangeVector(Minimum=<0 0 800>, Maximum=<1 1 1500>))
   &SolveForcesAndVelocity(Clamp Velocity=false, Limit Acceleration=false)
  ]
  rendererClasses [&NiagaraSpriteRendererProperties]
  renderers [{rendererClass=&NiagaraSpriteRendererProperties}]
 }
 WispySmoke004 {
  bEnabled=false
  emitterUpdateScript [
   &EmitterState(Life Cycle Mode=::NewEnumerator1, Loop Behavior=::NewEnumerator1, Loop Duration=2, UseLoopDelay=false)
   &SpawnRate(SpawnRate=15, Use Spawn Probability=false)
  ]
  particleSpawnScript [&InitializeParticle(
    Color=&ColorFromCurve(
     ColorCurve=DataInterface() {propertyValues {
       AlphaCurve {keys [<at 0 0 RCIM_Cubic>,<at 0.1064 0.5>,<at 0.8661 0.5>,<at 1 0 RCIM_Cubic>]}
       BlueCurve {keys [<at 0 1.5 RCIM_Cubic>,<at 1 1 RCIM_Cubic>]}
       GreenCurve {keys [<at 0 0 RCIM_Cubic>,<at 1 0 RCIM_Cubic>]}
       RedCurve {keys [<at 0 0.4181 RCIM_Cubic>,<at 1 0.3446 RCIM_Cubic>]}
      }}
     CurveIndex=$Emitter.NormalizedLoopAge
    )
    Lifetime Max=0.6
    Lifetime Min=0.4
    Lifetime Mode=::NewEnumerator1
    Sprite Size Max=&Lerp_Vector2(A=<-30 25>, B=<-25 30>, Alpha=&RandomRangeFloat(Minimum=0, Maximum=1))
    Sprite Size Min=&Lerp_Vector2(A=<20 15>, B=<20 15>, Alpha=&RandomRangeFloat(Minimum=0, Maximum=1))
    Sprite Size Mode=::NewEnumerator2
    UsePositionOffset=false
    Write Material Random=true
   )]
  particleUpdateScript [
   &InheritSourceMovement(
    Applied Position Scale=&Multiply_VectorByFloat(
     Vector=<1 1 1>
     Float=&FloatFromCurve(FloatCurve=DataInterface() {propertyValues {Curve {keys [<at 0 1>,<at 0.6 1>,<at 0.61 0>]}}}, CurveIndex=$Emitter.NormalizedLoopAge)
    )
    ApplyForce=false
    ApplyPosition=true
    ApplyPositionFalloff=false
   ) {enabled=false}
   &ParticleState()
   &ScaleSpriteSize(
    Non-Uniform Curve Scale=&RandomRangeVector2D(Minimum=<1 1>, Maximum=<1.2 1>)
    Non-Uniform Curve Sprite Scale=DataInterface() {propertyValues {
      XCurve {keys [{tangentMode=RCTM_User, time=0, value=0.9}, {tangentMode=RCTM_User, time=1, value=0.8}]}
      YCurve {keys [{tangentMode=RCTM_User, time=0, value=0.25}, <at 0.4 0.5>, {tangentMode=RCTM_User, time=1, value=2}]}
     }}
    Scale Sprite Size Mode=::NewEnumerator3
   )
   &SubUVAnimation(
    End Frame Range Override=30
    "Sprite Renderer "=DataInterface() {propertyValues {SpriteRenderer=&self:WispySmoke004_0.NiagaraSpriteRendererProperties_0}}
    Start Frame Range Override=&RandomRangeInt(Minimum=0, Maximum=5)
    SubUV Lookup Index:NiagaraFloat=&Lerp_Float(
     A=&RandomRangeFloat(Minimum=0, Maximum=0.2)
     B=&RandomRangeFloat(Minimum=0.3, Maximum=1)
     Alpha=$Particles.NormalizedAge
    )
    UseEndFrame=true
    UseStartFrame=true
   )
   &ScaleColor(
    Scale Alpha=&Multiply_Float(
     A=&RandomRangeFloat(Minimum=0.5, Maximum=1)
     B=&FloatFromCurve(
      FloatCurve=DataInterface() {propertyValues {Curve {keys [
          {arriveTangent=0.9483, interpMode=RCIM_Cubic, leaveTangent=0.9483, tangentMode=RCTM_User, time=0, value=0}
          {arriveTangent=2, interpMode=RCIM_Cubic, leaveTangent=2, time=0.1, value=0.05}
          {interpMode=RCIM_Cubic, tangentMode=RCTM_User, time=0.5, value=1}
          {arriveTangent=-5.5093, interpMode=RCIM_Cubic, leaveTangent=-5.5093, tangentMode=RCTM_User, time=1, value=0}
         ]}}}
      Scale Curve=&FloatFromCurve(
       FloatCurve=DataInterface() {propertyValues {Curve {keys [
           {tangentMode=RCTM_User, time=0, value=1}
           {arriveTangent=-0.0405, interpMode=RCIM_Cubic, leaveTangent=-0.0405, tangentMode=RCTM_User, time=1.5916, value=1}
           {arriveTangent=-2.5083, interpMode=RCIM_Cubic, leaveTangent=-2.5083, tangentMode=RCTM_User, time=2.4445, value=0}
          ]}}}
       CurveIndex=$Emitter.Age
      )
     )
    )
    ScaleA=true
    ScaleRGB=true
   )
   &SystemLocation(
    MaskSpawnGroup=false
    Offset=&RandomRangeVector(Minimum=<0 0 -30>, Maximum=<0 0 -50>, Random Seed=1, Fixed Random Seed=true)
   )
   &GravityForce(Gravity=<0 0 200>)
   &SolveForcesAndVelocity(Clamp Velocity=false, Limit Acceleration=false)
  ]
  rendererClasses [&NiagaraSpriteRendererProperties]
  renderers [{rendererClass=&NiagaraSpriteRendererProperties}]
 }
}
summary {
 systemName=NS_CombCrater_Pour
 emitters [
  {bEnabled=false, emitterName=Trail_Water, rendererClasses [&NiagaraRibbonRendererProperties]}
  {bEnabled=false, emitterName=Trail_Water001, rendererClasses [&NiagaraRibbonRendererProperties]}
  {bEnabled=false, emitterName=Trail_Water002, rendererClasses [&NiagaraRibbonRendererProperties]}
  {emitterName=Trail_Water003, rendererClasses [&NiagaraRibbonRendererProperties]}
  {bEnabled=false, emitterName=Trail_Water004, rendererClasses [&NiagaraRibbonRendererProperties]}
  {bEnabled=false, emitterName=WispySmoke002, rendererClasses [&NiagaraSpriteRendererProperties]}
  {bEnabled=false, emitterName=WispySmoke004, rendererClasses [&NiagaraSpriteRendererProperties]}
  {emitterName=Fire_Particle, rendererClasses [&NiagaraSpriteRendererProperties]}
 ]
 userVariables [
  User.FXScale {type=NiagaraFloat, defaultValue=1, description=""}
  User.Inherit Velocity {type=NiagaraFloat, defaultValue=0.15, description=""}
  User.Smoke Lifetime {type=NiagaraFloat, defaultValue=3, description=""}
 ]
}
