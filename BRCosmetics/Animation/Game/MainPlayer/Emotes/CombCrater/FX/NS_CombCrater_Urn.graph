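# Niagara system for the CombCrater emote's urn, producing a purple flame with glow and rising embers over roughly five to six seconds.
# Burst, glow and two flame-particle emitters sample a bone or socket on the urn skeleton, with flame offsets animated over loop age.
# An invisible PositionParent particle tracks the same point and feeds ember spawning through an attribute reader, with curl noise and drag shaping their motion.
# The glow adapts to time of day, platform and User.FXScale.
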NiagaraSystem /BRCosmetics/Animation/Game/MainPlayer/Emotes/CombCrater/FX/NS_CombCrater_Urn
dependencies {
 usedDataInterfaces [
  &NiagaraDataInterfacePlatformSet
  &NiagaraDataInterfaceCurve
  &NiagaraDataInterfaceColorCurve
  &NiagaraDataInterfaceParticleRead
  &NiagaraDataInterfaceSkeletalMesh
 ]
 usedDynamicInputs [&MaskFloatBySpawnGroup, &SimulationPosition]
 usedModules [
  &SystemState
  &self:SystemSpawnScript.SystemScriptSource.SystemScriptGraph.NiagaraNodeAssignment_0.SetVariables_#1
 ]
}
emitters {
 Embers002 {
  emitterUpdateScript [
   &EmitterState(
    Enable Distance Culling=true
    Enable Visibility Culling=true
    Life Cycle Mode=::NewEnumerator1
    Loop Behavior=::NewEnumerator1
    Loop Duration=5.5
    MaxDistance=1000
    Module.bApplySpawnCountScaleByDistance=true
    Module.bUseMaxDistance=true
    Module.bUseMinDistance=false
    Scalability Mode=::NewEnumerator1
    Scale Spawn Count=true
    Spawn Count Scale By Distance=DataInterface() {propertyValues {Curve {keys [<at 0 1>,<at 0.3 1>,<at 1 0>]}}}
    UseLoopDelay=false
   )
   &SpawnRate(
    Spawn Probability=&RandomRangeFloat(Minimum=0, Maximum=1, Evaluation Type=::NewEnumerator1)
    SpawnRate=&Lerp_Float(
     A=&RandomRangeFloat(Minimum=-200, Maximum=300, Evaluation Type=::NewEnumerator1)
     B=0
     Alpha=&ClampFloat(Float=&FN_LOD_Distance_Remap(Start Distance=200, End Distance=1200, Minimum Output=0, Maximum Output=1))
    )
    Use Spawn Probability=true
   ) {enabled=false}
   &SpawnParticlesFromOtherEmitter(
    Attribute Reader=DataInterface() {propertyValues {EmitterBinding {bindingMode=Other, emitterName=PositionParent}}}
    Spawn Rate=30
    Use Maximum Age=false
    Use Minimum Age=false
   )
   &SampleSkeletalMesh(AdditionalYawPitchRoll=false, OverrideRandomSeed=false, Skeletal Mesh:NiagaraDataInterfaceSkeletalMesh=$System.Urn Skeleton)
  ]
  particleSpawnScript [
   &InitializeParticle(
    Lifetime=&UniformRangedFloat(Minimum=0.3, Maximum=2.5)
    Sprite Size Mode=::NewEnumerator3
    Uniform Sprite Size=&MakeFloatFromVector2d(Vector 2D=&Vector2DFromFloat(Value=&FloatFromCurve(
       FloatCurve=DataInterface() {propertyValues {Curve {keys [<at 0 1.2>,<at 0.1257 0.9847>,<at 0.2996 0.5>,<at 1 0.3>]}}}
       CurveIndex=&UniformRangedFloat(Minimum=0, Maximum=1)
       Scale Curve=2
      )))
    UsePositionOffset=false
    Write Material Random=true
   ) {module=InitializeParticle001}
   &SampleParticlesFromOtherEmitter(Reject Particle If Sampled Index Is Invalid=false)
   &SphereLocation(
    Hemisphere Z=true
    MaskSpawnGroup=false
    Offset Coordinate Space:ENiagaraCoordinateSpace=::Simulation
    Override Seed=false
    Sphere Coordinate Space=::World
    Sphere Radius=2
    Surface Only=true
    UseNonUniformScale=true
    UseOffset=true
    UseSphereOrientation=false
   )
   &AddVelocity(Rotation Coordinate Space=::World)
  ]
  particleUpdateScript [
   &ParticleState()
   &InheritSourceMovement(ApplyForce=false, ApplyPosition=true, ApplyPositionFalloff=false)
   &AccelerationForce(Acceleration=&Multiply_Vector(
     A=&VectorFromFloat(Value=0.25)
     B=&UniformRangedVector(Minimum=<0 0 350>, Maximum=<0 0 500>)
    ))
   &Drag(
    Drag=&FloatFromCurve(FloatCurve=DataInterface() {propertyValues {Curve {keys [<at 0 0>,<at 0.1 0>,<at 1 1>]}}}, Scale Curve=50)
    Use Linear Drag=true
    Use Rotational Drag=false
   )
   /Niagara/Modules/Update/Forces/CurlNoiseForce(
    Noise Strength=&FloatFromCurve(
     FloatCurve=DataInterface() {propertyValues {Curve {keys [<at 0 1>,<at 0.1 0.25>,<at 0.1198 0.98>,<at 0.2002 0.21>,<at 0.2421 0.9996>,<at 1 0>]}}}
     Scale Curve=&RandomRangeFloat(Minimum=300, Maximum=500)
    )
    Pan Noise=true
    Pan Noise Field=<0 0 1>
   )
   &GravityForce(Gravity=&RandomRangeVector(Minimum=<0 0 0>, Maximum=<0 0 -250>))
   &SolveForcesAndVelocity(Clamp Velocity=false, Limit Acceleration=false)
   &ScaleSpriteSizeBySpeed(
    Max Scale Factor=<2 5>
    Sample Scale Factor By Curve=true
    Scale Factor Curve=DataInterface() {propertyValues {Curve {keys [
        <at -0.0077 0.0765 RCIM_Cubic>
        {arriveTangent=1.2463, interpMode=RCIM_Cubic, leaveTangent=1.2463, tangentMode=RCTM_User, time=0.9985, value=0.9549}
       ]}}}
    Velocity Threshold=400
   )
   &ScaleColor(
    Scale Alpha=&FloatFromCurve(FloatCurve=DataInterface() {propertyValues {Curve {keys [<at 0 1>,<at 1 0>]}}})
    Scale RGB=&MakeVectorFromLinearColorRGB(Color=&ColorFromCurve(ColorCurve=DataInterface() {propertyValues {
        AlphaCurve {keys [<at 0 1 RCIM_Cubic>,<at 0.7107 1>]}
        BlueCurve {keys [<at 0 15 RCIM_Cubic>,<at 0.7114 0.9384 RCIM_Cubic>]}
        GreenCurve {keys [<at 0 0 RCIM_Cubic>,<at 0.7114 0 RCIM_Cubic>]}
        RedCurve {keys [<at 0 7.0975 RCIM_Cubic>,<at 0.7114 1 RCIM_Cubic>]}
       }}))
    ScaleA=true
    ScaleRGB=true
   )
   &ScaleSpriteSizeBySpeed(Max Scale Factor=<1 3>, Min Scale Factor=<0.5 0.5>, Velocity Threshold=350) {module=ScaleSpriteSizeBySpeed001}
  ]
  rendererClasses [&NiagaraSpriteRendererProperties]
  renderers [{rendererClass=&NiagaraSpriteRendererProperties}]
 }
 Fire_Burst {
  emitterUpdateScript [
   &EmitterState(Life Cycle Mode=::NewEnumerator1, Loop Behavior=::NewEnumerator1, Loop Duration=6, UseLoopDelay=false)
   &SpawnBurst_Instantaneous(Spawn Count=2, Use Spawn Probability=false, UseLoopCountLimit=false)
  ]
  particleSpawnScript [
   /SetVariables(Particles.VisibilityTag=&MaskIntBySpawnGroup(Masked Int=0, Passthrough Int=1, Spawn Group Mask=0)) {isSetParameters=true}
   &InitializeParticle(
    Color=<rgba 0.2959 0.1334 0.6458 1.5>
    Lifetime Max=1
    Lifetime Min=0.7
    Lifetime Mode=::NewEnumerator1
    Position Offset=&MaskVectorBySpawnGroup(Masked Vector=<0 0 13.5>, Passthrough Vector=<0 0 20>)
    Sprite Rotation Mode=::NewEnumerator2
    Sprite Size=&MaskVector2dBySpawnGroup(Masked Vector2D=<30 30>, Passthrough Vector2D=<30 40>, Spawn Group Mask=0)
    Sprite Size Mode=::NewEnumerator4
    UsePositionOffset=true
    Write Material Random=true
   )
   &CameraOffset(Camera Offset Amount=&FN_PickFloatDependOnPlatform(TargetPlatformActive=0, TargetPlatformNoActive=2))
   &AddVelocity(Velocity=&MaskVectorBySpawnGroup(Masked Vector=<0 0 5>, Passthrough Vector=<0 0 0>), Velocity Speed Scale=2)
  ]
  particleUpdateScript [
   &ParticleState()
   &SkeletalMeshLocation(
    Bone / Socket Index=1
    Bone Sampling Mode=::NewEnumerator4
    MaskSpawnGroup=false
    Sampled Position Offset=<0 0 8>
    Skeletal Mesh=$System.Urn Skeleton
   )
   &SystemLocation(
    MaskSpawnGroup=false
    Offset=&MaskVectorBySpawnGroup(Masked Vector=<0 0 10>, Passthrough Vector=<0 0 6>)
    Offset Coordinate Space=::Simulation
   ) {enabled=false}
   /SetVariables(Particles.SubImageIndex=&FloatFromCurve(FloatCurve=DataInterface() {propertyValues {Curve {keys [<at 0 5>,<at 1 15>]}}})) {isSetParameters=true, module=SetVariables_#2}
   &ScaleSpriteSize(Scale Factor=&Vector2DFromFloat(Value=&FloatFromCurve(
      FloatCurve=DataInterface() {propertyValues {Curve {keys [
          {interpMode=RCIM_Cubic, leaveTangent=3.0564, tangentMode=RCTM_Break, time=0, value=0}
          {arriveTangent=1.1454, interpMode=RCIM_Cubic, leaveTangent=1.1454, tangentMode=RCTM_User, time=0.3, value=0.8}
          {arriveTangent=-1.5127, interpMode=RCIM_Cubic, tangentMode=RCTM_Break, time=1, value=0.6}
         ]}}}
      Scale Curve=0.8
     )))
   &ScaleColor(
    Scale Alpha=&FloatFromCurve(FloatCurve=DataInterface() {propertyValues {Curve {keys [<at 0 0 RCIM_Cubic>,<at 0.3 1 RCIM_Cubic>,<at 1 0 RCIM_Cubic>]}}})
    Scale RGB=&MakeVectorFromLinearColorRGB(Color=&ColorFromCurve(ColorCurve=DataInterface() {propertyValues {
        AlphaCurve {keys [<at 0 1 RCIM_Cubic>]}
        BlueCurve {keys [<at 0 1 RCIM_Cubic>,<at 1 1 RCIM_Cubic>]}
        GreenCurve {keys [<at 0 0.0048 RCIM_Cubic>,<at 1 1 RCIM_Cubic>]}
        RedCurve {keys [<at 0 0.682 RCIM_Cubic>,<at 1 1 RCIM_Cubic>]}
       }}))
    ScaleA=true
    ScaleRGB=true
   )
   &SolveForcesAndVelocity(Clamp Velocity=false, Limit Acceleration=false)
  ]
  rendererClasses [&NiagaraSpriteRendererProperties]
  renderers [
   {rendererClass=&NiagaraSpriteRendererProperties}
   {rendererClass=&NiagaraSpriteRendererProperties, rendererIndex=1}
  ]
 }
 Fire_Glow {
  emitterUpdateScript [
   &EmitterState(Life Cycle Mode=::NewEnumerator1, Loop Behavior=::NewEnumerator1, Loop Duration=5.5, UseLoopDelay=false)
   &SpawnBurst_Instantaneous(Use Spawn Probability=false, UseLoopCountLimit=false)
  ]
  particleSpawnScript [
   &InitializeParticle(
    Color=<rgba 1 1 1 0.2>
    Lifetime=6
    Mesh Scale Mode=::NewEnumerator3
    Sprite Size Mode=::NewEnumerator3
    Uniform Sprite Size=&Multiply_Float(A=38.4, B=$User.FXScale)
    UsePositionOffset=true
    Write Material Random=true
   )
   &FN_Sprite_Attributes(
    Color=<rgba 0.5461 0.2645 1 0.5>
    Create Color Attribute=true
    Create Position Attribute=true
    Create SpriteSize Attribute=true
    Lifetime=6
    Position:NiagaraPosition=<5 0 0>
    SpriteSize=&Vector2DFromFloat(Value=&Multiply_Float(A=15.5, B=$User.FXScale))
    Write to Module Namespace=true
   )
  ]
  particleUpdateScript [
   &ParticleState()
   &SkeletalMeshLocation(
    Bone / Socket Index=1
    Bone Sampling Mode=::NewEnumerator4
    MaskSpawnGroup=false
    Sampled Position Offset=<0 0 10>
    Skeletal Mesh=$System.Urn Skeleton
   )
   &ScaleColor(
    Scale Alpha=&FN_PickFloatDependOnPlatform(TargetPlatformActive=1, TargetPlatformNoActive=0.5)
    Scale RGB=&VectorFromFloat(Value=&FNTimeOfDayScalar(Night Scale=0.8, Morning Scale=1, Evening Scale=0.5))
    ScaleA=true
    ScaleRGB=true
   )
   &ScaleColor(
    Scale Alpha=&FloatFromCurve(FloatCurve=DataInterface() {propertyValues {Curve {keys [<at 0 1>,<at 0.7 1>,<at 1 0>]}}})
    ScaleA=true
    ScaleRGB=true
   ) {module=ScaleColor001}
   /SetVariables(Particles.FN_Sprite_Attributes.Color:LinearColor=&ColorFromCurve(ColorCurve=DataInterface() {propertyValues {
       AlphaCurve {keys [<at 0 0.5 RCIM_Cubic>,<at 0.7776 1>,<at 1 0 RCIM_Cubic>]}
       BlueCurve {keys [<at 0 1 RCIM_Cubic>,<at 1 1 RCIM_Cubic>]}
       GreenCurve {keys [<at 0 0.2627 RCIM_Cubic>,<at 1 0.2627 RCIM_Cubic>]}
       RedCurve {keys [<at 0 0.5451 RCIM_Cubic>,<at 1 0.5451 RCIM_Cubic>]}
      }})) {isSetParameters=true}
  ]
  rendererClasses [&NiagaraSpriteRendererProperties]
  renderers [
   {rendererClass=&NiagaraSpriteRendererProperties}
   {rendererClass=&NiagaraSpriteRendererProperties, rendererIndex=1}
  ]
 }
 Fire_Particle {
  emitterUpdateScript [
   &EmitterState(Life Cycle Mode=::NewEnumerator1, Loop Behavior=::NewEnumerator1, Loop Duration=5.25, UseLoopDelay=false)
   &SpawnRate(SpawnRate=15, Use Spawn Probability=false)
  ]
  particleSpawnScript [
   &InitializeParticle(
    Color=<rgba 0.8655 0.3636 2 1>
    Lifetime Max=1
    Lifetime Min=0.75
    Lifetime Mode=::NewEnumerator1
    Sprite Rotation Mode=::NewEnumerator2
    Sprite Size Mode=::NewEnumerator3
    Uniform Sprite Size=&RandomRangeFloat(Minimum=20, Maximum=30)
    UsePositionOffset=false
    Write Material Random=true
   )
   &CameraOffset(Camera Offset Amount=&FN_PickFloatDependOnPlatform(TargetPlatformActive=0, TargetPlatformNoActive=2))
  ]
  particleUpdateScript [
   &ParticleState()
   &SkeletalMeshLocation(
    Bone / Socket Index=1
    Bone Sampling Mode=::NewEnumerator4
    MaskSpawnGroup=false
    Position Sampling=::NewEnumerator1
    Sampled Position Offset=&MakeVector(
     X=0
     Y=&FloatFromCurve(
      FloatCurve=DataInterface() {propertyValues {Curve {keys [<at 0 0>,<at 0.5 0>,<at 0.6 -2>,<at 0.7 -2>,<at 0.922 0>,<at 1 0>]}}}
      CurveIndex=$Emitter.NormalizedLoopAge
     )
     Z=12
    )
    Skeletal Mesh=$System.Urn Skeleton
   )
   /SetVariables(Particles.SubImageIndex=&FloatFromCurve(FloatCurve=DataInterface() {propertyValues {Curve {keys [<at 0 5>,<at 1 15>]}}})) {isSetParameters=true, module=SetVariables_#2}
   &ScaleSpriteSize(Scale Factor=&Vector2DFromFloat(Value=&FloatFromCurve(
      FloatCurve=DataInterface() {propertyValues {Curve {keys [
          {interpMode=RCIM_Cubic, leaveTangent=0.1472, tangentMode=RCTM_Break, time=0, value=0}
          {arriveTangent=2.3722, interpMode=RCIM_Cubic, leaveTangent=2.3722, tangentMode=RCTM_User, time=0.3, value=0.8}
          {arriveTangent=-1.5127, interpMode=RCIM_Cubic, tangentMode=RCTM_Break, time=1, value=0.6}
         ]}}}
      Scale Curve=0.8
     )))
   &ScaleColor(
    Scale Alpha=&FloatFromCurve(FloatCurve=DataInterface() {propertyValues {Curve {keys [<at 0 0 RCIM_Cubic>,<at 0.3 1 RCIM_Cubic>,<at 1 0 RCIM_Cubic>]}}})
    Scale RGB=&MakeVectorFromLinearColorRGB(Color=&ColorFromCurve(ColorCurve=DataInterface() {propertyValues {
        AlphaCurve {keys [<at 0 1 RCIM_Cubic>,<at 1 0 RCIM_Cubic>]}
        BlueCurve {keys [<at 0 3 RCIM_Cubic>,<at 1 0.25 RCIM_Cubic>]}
        GreenCurve {keys [<at 0 0.597 RCIM_Cubic>,<at 1 0.25 RCIM_Cubic>]}
        RedCurve {keys [<at 0 2.5359 RCIM_Cubic>,<at 1 0.25 RCIM_Cubic>]}
       }}))
    ScaleA=true
    ScaleRGB=true
   )
   &GravityForce(Gravity=&RandomRangeVector(Minimum=<0 0 25>, Maximum=<0 0 100>))
   &SolveForcesAndVelocity(Clamp Velocity=true, Limit Acceleration=false)
  ]
  rendererClasses [&NiagaraSpriteRendererProperties]
  renderers [{rendererClass=&NiagaraSpriteRendererProperties}]
 }
 Fire_Particle001 {
  emitterUpdateScript [
   &EmitterState(Life Cycle Mode=::NewEnumerator1, Loop Behavior=::NewEnumerator1, Loop Duration=5.25, UseLoopDelay=false)
   &SpawnRate(SpawnRate=25, Use Spawn Probability=false)
  ]
  particleSpawnScript [
   &InitializeParticle(
    Color=<rgba 0.8655 0.3636 2 1>
    Lifetime Max=1
    Lifetime Min=0.75
    Lifetime Mode=::NewEnumerator1
    Sprite Rotation Mode=::NewEnumerator2
    Sprite Size Mode=::NewEnumerator3
    Uniform Sprite Size=&RandomRangeFloat(Minimum=10, Maximum=15)
    UsePositionOffset=false
    Write Material Random=true
   )
   &CameraOffset(Camera Offset Amount=&FN_PickFloatDependOnPlatform(TargetPlatformActive=0, TargetPlatformNoActive=2))
  ]
  particleUpdateScript [
   &ParticleState()
   &SkeletalMeshLocation(
    Bone / Socket Index=1
    Bone Sampling Mode=::NewEnumerator4
    MaskSpawnGroup=false
    Sampled Position Offset=&MakeVector(
     X=0
     Y=&FloatFromCurve(
      FloatCurve=DataInterface() {propertyValues {Curve {keys [<at 0 0>,<at 0.5 0>,<at 0.6 -2>,<at 0.7 -2>,<at 0.922 0>,<at 1 0>]}}}
      CurveIndex=$Emitter.NormalizedLoopAge
     )
     Z=&FloatFromCurve(
      FloatCurve=DataInterface() {propertyValues {Curve {keys [<at 0 10>,<at 0.5 10>,<at 0.5512 12>,<at 0.6512 12>,<at 0.7512 10>,<at 0.8512 10>]}}}
      CurveIndex=$Emitter.NormalizedLoopAge
     )
    )
    Skeletal Mesh=$System.Urn Skeleton
   )
   /SetVariables(Particles.SubImageIndex=&FloatFromCurve(FloatCurve=DataInterface() {propertyValues {Curve {keys [<at 0 5>,<at 1 15>]}}})) {isSetParameters=true, module=SetVariables_#2}
   &ScaleSpriteSize(Scale Factor=&Vector2DFromFloat(Value=&FloatFromCurve(
      FloatCurve=DataInterface() {propertyValues {Curve {keys [
          {interpMode=RCIM_Cubic, leaveTangent=0.1472, tangentMode=RCTM_Break, time=0, value=0}
          {arriveTangent=2.3722, interpMode=RCIM_Cubic, leaveTangent=2.3722, tangentMode=RCTM_User, time=0.3, value=0.8}
          {arriveTangent=-1.5127, interpMode=RCIM_Cubic, tangentMode=RCTM_Break, time=1, value=0.6}
         ]}}}
      Scale Curve=0.8
     )))
   &ScaleColor(
    Scale Alpha=&FloatFromCurve(FloatCurve=DataInterface() {propertyValues {Curve {keys [<at 0 0 RCIM_Cubic>,<at 0.3 1 RCIM_Cubic>,<at 1 0 RCIM_Cubic>]}}})
    Scale RGB=&MakeVectorFromLinearColorRGB(Color=&ColorFromCurve(ColorCurve=DataInterface() {propertyValues {
        AlphaCurve {keys [<at 0 1 RCIM_Cubic>,<at 1 0 RCIM_Cubic>]}
        BlueCurve {keys [<at 0 3 RCIM_Cubic>,<at 1 0.25 RCIM_Cubic>]}
        GreenCurve {keys [<at 0 0.597 RCIM_Cubic>,<at 1 0.25 RCIM_Cubic>]}
        RedCurve {keys [<at 0 2.5359 RCIM_Cubic>,<at 1 0.25 RCIM_Cubic>]}
       }}))
    ScaleA=true
    ScaleRGB=true
   )
   &GravityForce(Gravity=&RandomRangeVector(Minimum=<-5 -5 25>, Maximum=<5 5 100>))
   &SolveForcesAndVelocity(Clamp Velocity=false, Limit Acceleration=false)
  ]
  rendererClasses [&NiagaraSpriteRendererProperties]
  renderers [{rendererClass=&NiagaraSpriteRendererProperties}]
 }
 PositionParent {
  emitterUpdateScript [
   &EmitterState()
   &SpawnBurst_Instantaneous(Use Spawn Probability=false, UseLoopCountLimit=false)
  ]
  particleSpawnScript [&InitializeParticle(Color=<rgba 1 1 1 0>, Lifetime=5, Sprite Size Mode=::NewEnumerator3, UsePositionOffset=false, Write Material Random=true)]
  particleUpdateScript [
   &ParticleState()
   &SkeletalMeshLocation(
    Bone / Socket Index=1
    Bone Sampling Mode=::NewEnumerator4
    MaskSpawnGroup=false
    Orientation Sampling=::NewEnumerator1
    Sampled Position Offset=<0 0 12>
    Skeletal Mesh=$System.Urn Skeleton
   )
  ]
  rendererClasses []
  renderers []
 }
}
summary {
 systemName=NS_CombCrater_Urn
 emitters [
  {emitterName=Fire_Burst, rendererClasses [&NiagaraSpriteRendererProperties]}
  {emitterName=Fire_Glow, rendererClasses [&NiagaraSpriteRendererProperties]}
  {emitterName=Embers002, rendererClasses [&NiagaraSpriteRendererProperties]}
  {emitterName=PositionParent, rendererClasses []}
  {emitterName=Fire_Particle, rendererClasses [&NiagaraSpriteRendererProperties]}
  {emitterName=Fire_Particle001, rendererClasses [&NiagaraSpriteRendererProperties]}
 ]
 userVariables [
  User.bIsFrontEndPreview {type=NiagaraBool}
  User.FXScale {type=NiagaraFloat, defaultValue=1, description=""}
 ]
}
